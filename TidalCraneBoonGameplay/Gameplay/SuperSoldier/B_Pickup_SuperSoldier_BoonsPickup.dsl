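; Super Soldier boon pickup effect Blueprint that spins and color-codes the item and animates the collecting pawn, reusing Fire Petal boon assets.
; BeginPlay adds a rotating movement component to the parent pickup actor, driving its static mesh, and tints the idle Niagara effect from a rarity-indexed color array.
; It strips the inherited spotlight and dark background components.
; On pickup, it plays the Fire Petal consume montage on the pawn at increased speed.
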(event EventOnPickedUp
  (params (PickupTarget :type Fort%20Pawn%20Object%20Reference))
  (|CallParentFunction PickupTarget)
  (Animation|Montage|PlayMontage (Variables|/Script/Engine.Character|GetMesh PickupTarget) "/FirePetalBoonsGameplay/Animations/CMM_Proto_FirePetalBoon_Consume.CMM_Proto_FirePetalBoon_Consume" 2.0))

(event EventBeginPlay
  (bind _parentpickupactor (Variables|PickupProperties|GetParentPickupActor))
  (|CallParentFunction)
  (bind _returnvalue (AddComponent|NODE_AddRotatingMovementComponent-0 :self _parentpickupactor))
  (Components|Tick|SetTickGroup _returnvalue "TG_DuringPhysics")
  (Niagara|SetNiagaraVariable(LinearColor) (Variables|Default|GetNS_FirePetal_Boons_Idle) "Global Color" (Utilities|Array|Get(aref) (Variables|Default|GetNewVar) (Variables|FortPickupsParent|GetPickupRarityLevel)))
  (Components|Movement|SetUpdatedComponent _returnvalue (Class|FortPickup|GetPickupStaticMeshComponent _parentpickupactor))
  (Components|DestroyComponent (Variables|Default|GetspotlightComp)))

(event Collision|EventActorBeginOverlap
  (params (OtherActor :type Actor%20Object%20Reference)))

(event EventTick
  (params (DeltaSeconds :type Float%20%28single-precision%29)))

(fn ConstructionScript
  (|CallParentFunction)
  (Components|DestroyComponent (Variables|Default|GetBG_Dark)))
